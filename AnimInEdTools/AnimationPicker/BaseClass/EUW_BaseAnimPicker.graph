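# Base Editor Utility Widget for animation picker panels in the AnimInEdTools plugin, tied to an animatable creature's Control Rig in Sequencer.
# On the loaded event it caches the rig binding, Control Rig and focused Level Sequence, stores the owning picker and creature, then prints the creature's name.
# It exposes an empty canvas as the primary panel, returns no selectable elements and leaves Update From Scene unimplemented.
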EditorUtilityWidgetBlueprint /AnimInEdTools/AnimationPicker/BaseClass/EUW_BaseAnimPicker
graphs [
 EventGraph {nodes [
   n1 = Event Widget has loaded!() -> (
    OutputDelegate
    then
    Control Rig Proxy Binding:Control Rig Sequencer Binding Proxy Structure
    Animatable Creature:Animatable Creature Object Reference
    Anim Picker EUW:Animation Picker Object Reference
   ) {is=title K2Node_Event}
   > Initialize!(execute, self, Control Rig Proxy Binding:Control Rig Sequencer Binding Proxy Structure=@n1.Control Rig Proxy Binding) -> (then)
   > Set Anim Picker EUW(Anim Picker EUW:Animation Picker Object Reference=@n1.Anim Picker EUW) -> (Output_Get:Animation Picker Object Reference)
   > n2 = Set Animatable Creature(Animatable Creature:Animatable Creature Object Reference=@n1.Animatable Creature) -> (Output_Get:Animatable Creature Object Reference)
   > PrintString(InString=Append(A=".\r\nLoaded ", B=GetDisplayName(Object=@n2)))
   Event PreConstruct()
   Event Construct()
   Event Tick()
   Event Update From Scene() {is=title K2Node_Event}
  ]}
 Get Primary Panel {nodes [
   Get Primary Panel() {is=title K2Node_FunctionEntry}
   > Return Node!(execute, Panel:Panel Widget Object Reference=Get MainCanvasPanel() -> (MainCanvasPanel:Canvas Panel Object Reference))
  ]}
 Get all Selectable Elements {nodes [
   Get all Selectable Elements() {is=title K2Node_FunctionEntry}
   > Return Node!(execute, All Elements:Array of EUW Picker Element Object References)
  ]}
 Initialize {nodes [
   n1 = Initialize!() -> (then, Control Rig Proxy Binding:Control Rig Sequencer Binding Proxy Structure) {is=title K2Node_FunctionEntry}
   > n2 = Set Control Rig Proxy Binding(Control Rig Proxy Binding:Control Rig Sequencer Binding Proxy Structure=@n1) -> (Output_Get:Control Rig Sequencer Binding Proxy Structure)
   > Set Control Rig(Control Rig:Control Rig Object Reference=Break Control Rig Sequencer Binding Proxy!(ControlRigSequencerBindingProxy:"Control Rig Sequencer Binding Proxy Structure (by ref)"=@n2) -> (
     Proxy:Movie Scene Binding Proxy Structure
     ControlRig:Control Rig Object Reference
     Track:Movie Scene Control Rig Parameter Track Object Reference
    ) {is=title K2Node_BreakStruct}.ControlRig) -> (Output_Get:Control Rig Object Reference)
   > n3 = GetFocusedLevelSequence!(execute) -> (then, ReturnValue:Level Sequence Object Reference)
   > Set LevelSequencer(LevelSequencer:Level Sequence Object Reference=@n3) -> (Output_Get:Level Sequence Object Reference)
  ]}
]
variables [
 Anim Picker EUW {type=Animation_Picker_C, default=null}
 Animatable Creature {type=AnimatableCreature_C, default=null}
 Control Rig {type=ControlRig, default=null}
 Control Rig Proxy Binding {
  type=ControlRigSequencerBindingProxy
  default {controlRig=null, track=null, proxy {bindingId=#1, sequence=null}}
 }
 Frame Number {type=FrameNumber, default {value=0}}
 LevelSequencer {type=LevelSequence, default=null}
]
widgets {
 info {rootWidgetClass=&CanvasPanel, widgetCount=1}
 widgets [MainCanvasPanel {widgetClassPath=&CanvasPanel, bIsVariable=true, widget=MainCanvasPanel}]
}
